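; BuzzCool manager actor that runs Living World encounters for gameplay abilities, starts an ambient encounter at BeginPlay and stores the pawn with the donut.
; Encounters are keyed by ability class, plus avatar name unless single-instance, so restarting an ability resumes its existing encounter.
; Ability end pauses the encounter and optionally ends single-instance ones.
; Last-actor death cleans up only paused encounters, ending them and destroying their volumes.
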(event Custom|SetPawnWithDonut
  (params (PawnWithDonut :type Fort%20Player%20Pawn%20Object%20Reference))
  (Variables|Default|SetPawnWithDonut PawnWithDonut))

(event Custom|ClearPawnWithDonut
  (Variables|Default|SetPawnWithDonut 0))

(event Custom|AbilityEnding
  (params (ExecutingAbility :type Gameplay%20Ability%20Object%20Reference))
  (bind _activeencountermap (Variables|Default|GetActiveEncounterMap))
  (bind _stringidentifier (CallFunction|GetStringIdentifier ExecutingAbility))
  (Utilities|IsValid (Utilities|Map|Find _activeencountermap _stringidentifier)
    (:Is%20Valid
      (Class|FortAthenaLivingWorldEncounterInstance|PauseEncounter (Utilities|Map|Find _activeencountermap _stringidentifier))
      (Utilities|Map|Add (Variables|Default|GetPausedEncounterMap) (Utilities|Map|Find _activeencountermap _stringidentifier) _stringidentifier)
      (if (and (ScalableFloat|GetValueasBool (Variables|Default|GetEndEncounterWithAbility)) (Utilities|Set|ContainsItem (Variables|Default|GetSingleInstanceAbilities) (Utilities|GetClass ExecutingAbility)))
        (CallFunction|LastActorOfEncounterDied (Utilities|Map|Find _activeencountermap _stringidentifier))))))

(event Custom|LastActorOfEncounterDied
  (params (EncounterInstance :type Fort%20Athena%20Living%20World%20Encounter%20Instance%20Object%20Reference))
  (bind _pausedencountermap (Variables|Default|GetPausedEncounterMap))
  (bind (_value _returnvalue) (Utilities|Map|Find _pausedencountermap EncounterInstance))
  (bind _encountervolumemap (Variables|Default|GetEncounterVolumeMap))
  (bind _value_2 (Utilities|Map|Find _encountervolumemap EncounterInstance))
  (if _returnvalue
    (Utilities|Map|Remove (Variables|Default|GetActiveEncounterMap) _value)
    (Utilities|Map|Remove _pausedencountermap EncounterInstance)
    (Class|BPFLLagerCommunFunctions|EndActivityEncounter (Fortnite|Athena|GameState|GetGameStateAthena) EncounterInstance)
    (Utilities|IsValid _value_2
      (:Is%20Valid
        (Actor|DestroyActor _value_2)
        (Utilities|Map|Remove _encountervolumemap EncounterInstance)
        (EventDispatchers|UnbindEventfromOnLastActorDied EncounterInstance (EventDispatchers|CreateEvent :!event Custom|LastActorOfEncounterDied))))))

(event Custom|StartEncounterWithManager
  (params (ExecutingAbility :type Gameplay%20Ability%20Object%20Reference) (EncounterInstance :type Fort%20Athena%20Living%20World%20Encounter%20Soft%20Object%20Reference) (LivingWorldVolume :type Fort%20Athena%20Living%20World%20Volume%20Object%20Reference))
  (bind _activeencountermap (Variables|Default|GetActiveEncounterMap))
  (bind _stringidentifier (CallFunction|GetStringIdentifier ExecutingAbility))
  (Utilities|IsValid (Utilities|Map|Find _activeencountermap _stringidentifier)
    (:Is%20Valid
      (Utilities|Map|Remove (Variables|Default|GetPausedEncounterMap) (Utilities|Map|Find _activeencountermap _stringidentifier))
      (Class|FortAthenaLivingWorldEncounterInstance|ResumeEncounter (Utilities|Map|Find _activeencountermap _stringidentifier)))
    (:Is%20Not%20Valid
      (bind (_isvalid _returnvalue) (Default|StartActivityEncounter (Fortnite|Athena|GameState|GetGameStateAthena) LivingWorldVolume self EncounterInstance))
      (Utilities|Map|Add (Variables|Default|GetActiveEncounterMap) _stringidentifier _returnvalue)
      (Utilities|Map|Add (Variables|Default|GetEncounterVolumeMap) _returnvalue LivingWorldVolume)
      (EventDispatchers|AssignOnLastActorDied _returnvalue (EventDispatchers|CreateEvent :!event Custom|LastActorOfEncounterDied)))))

(event EventBeginPlay
  (Variables|Default|GetAmbientEncounterType
    (:then
      (bind _asfort_athena_living_world_volume (Utilities|Casting|CastToFortAthenaLivingWorldVolume (Variables|/Script/Engine.ChildActorComponent|GetChildActor (Variables|Default|GetLWV_BuzzCool_Ambient))))
      (Class|FortAthenaLivingWorldPointProviderInterface|EnablePointProvider_DEPRECATED _asfort_athena_living_world_volume)
      (bind (_isvalid _returnvalue) (Default|StartActivityEncounter (Fortnite|Athena|GameState|GetGameStateAthena) _asfort_athena_living_world_volume self (Variables|Default|GetAmbientEncounterType)))
      (Variables|Default|SetAmbientEncounterRef _returnvalue))
    (:else)))

(event Collision|EventActorBeginOverlap
  (params (OtherActor :type Actor%20Object%20Reference)))

(event EventTick
  (params (DeltaSeconds :type Float%20%28single-precision%29)))

(fn FindPausedEncounter
  (params (ExecutingAbility :type Gameplay%20Ability%20Object%20Reference))
  (bind _activeencountermap (Variables|Default|GetActiveEncounterMap))
  (bind _encountervolumemap (Variables|Default|GetEncounterVolumeMap))
  (bind _stringidentifier (CallFunction|GetStringIdentifier ExecutingAbility))
  (Utilities|IsValid (Utilities|Map|Find _activeencountermap _stringidentifier)
    (:Is%20Valid
      (Utilities|IsValid (Utilities|Map|Find _encountervolumemap (Utilities|Map|Find _activeencountermap _stringidentifier))
        (:Is%20Valid
          (Utilities|Map|Remove (Variables|Default|GetPausedEncounterMap) (Utilities|Map|Find _activeencountermap _stringidentifier))
          (return (Utilities|Map|Find _encountervolumemap (Utilities|Map|Find _activeencountermap _stringidentifier)) true))
        (:Is%20Not%20Valid
          (return 0 false))))))

(fn GetStringIdentifier
  (params (GameplayAbility :type Gameplay%20Ability%20Object%20Reference))
  (bind _returnvalue (Utilities|GetClass GameplayAbility))
  (bind _returnvalue_2 (Utilities|GetClassDisplayName _returnvalue))
  (if (Utilities|Set|ContainsItem (Variables|Default|GetSingleInstanceAbilities) _returnvalue)
    (return _returnvalue_2)
    (else
      (return (Utilities|String|Append _returnvalue_2 (Utilities|GetDisplayName (Ability|GetAvatarActorfromActorInfo GameplayAbility)))))))

(fn IsEncounterAbilityEnding
  (params (Encounter :type Fort%20Athena%20Living%20World%20Encounter%20Instance%20Object%20Reference))
  (return false))

(fn ConstructionScript)
